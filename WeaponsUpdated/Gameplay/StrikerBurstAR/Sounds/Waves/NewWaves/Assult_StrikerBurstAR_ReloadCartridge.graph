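# Sound cue for the Striker Burst AR reload's magazine insert foley.
# A Random node picks one of three equally weighted InsertMag wave variations, excluded from branch culling, and feeds it through a Modulator at the root.
# It uses the weapon reload attenuation and weapon foley sound class, with reduced volume.
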SoundCue /WeaponsUpdated/Gameplay/StrikerBurstAR/Sounds/Waves/NewWaves/Assult_StrikerBurstAR_ReloadCartridge
root=Modulator_0
nodes [
 {id=Modulator_0, children [Random_1]}
 {id=Random_1, children [WavePlayer_0,WavePlayer_1,WavePlayer_2], properties {bShouldExcludeFromBranchCulling=true, weights [1,1,1]}}
 {
  id=WavePlayer_0
  properties {soundWaveAssetPtr=&/Assult_StrikerBurstAR_InsertMag_01}
 }
 {
  id=WavePlayer_1
  properties {soundWaveAssetPtr=&/Assult_StrikerBurstAR_InsertMag_02}
 }
 {
  id=WavePlayer_2
  properties {soundWaveAssetPtr=&/Assult_StrikerBurstAR_InsertMag_03}
 }
]
properties {
 attenuationSettings=&Fort_Weapon_Reload_Att
 maxDistance=2250
 soundClassObject=&Fort_WeaponFoley
 volumeMultiplier=0.7
}
